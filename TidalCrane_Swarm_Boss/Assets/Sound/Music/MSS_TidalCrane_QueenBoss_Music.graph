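# Adaptive stereo MetaSound music for the TidalCrane Queen boss, driven by Boss.Phase.
# A Harmonix metronome quantizes phase changes, sequencing looping percussion/other stem pairs for three phases, transition stingers and an ending.
# The mix is scaled by fade-in, distance falloff, pause fade and ducking from a weapon-fire modulator.
# It finishes when the ending completes or a pause outlasts the fade-out.
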MetaSoundSource /TidalCrane_Swarm_Boss/Assets/Sound/Music/MSS_TidalCrane_QueenBoss_Music
inputs [
 Boss.Phase {type=Int32}
 DuckingStrength {type=Float}
 FadeInTime {type=Time}
 FadeOutTime {type=Time}
 MasterVolume {type=Float}
 Mix.Ending {type=Float}
 Mix.Phase0.Else {type=Float}
 Mix.Phase0.Master {type=Float}
 Mix.Phase0.Perc {type=Float}
 Mix.Phase1.Else {type=Float}
 Mix.Phase1.Master {type=Float}
 Mix.Phase1.Perc {type=Float}
 Mix.Phase2.Else {type=Float}
 Mix.Phase2.Master {type=Float}
 Mix.Phase2.Perc {type=Float}
 Mix.Transition0to1 {type=Float}
 Mix.Transition1to2 {type=Float}
 Music.BeatsPerBar {type=Int32}
 Music.Ending {type=WaveAsset}
 Music.Phase0.Else {type=WaveAsset}
 Music.Phase0.Perc {type=WaveAsset}
 Music.Phase1.Else {type=WaveAsset}
 Music.Phase1.Perc {type=WaveAsset}
 Music.Phase2.Else {type=WaveAsset}
 Music.Phase2.Perc {type=WaveAsset}
 Music.Tempo {type=Float}
 Music.Transition0to1 {type=WaveAsset}
 Music.Transition1to2 {type=WaveAsset}
 MusicEncounter.Attenuation.Volume.FalloffDistance {type=Float}
 MusicEncounter.Attenuation.Volume.InnerRadius {type=Float}
 MusicEncounter.bIsPaused {type=Bool}
 UE.Attenuation.Distance {type=Float}
 UE.Source.OnPlay {type=Trigger}
]
interfaces [UE.Attenuation:1.0, UE.OutputFormat.Stereo:1.0, UE.Source.OneShot:1.0, UE.Source:1.1]
outputs [
 "UE.OutputFormat.Stereo.Audio:0" {type=Audio}
 "UE.OutputFormat.Stereo.Audio:1" {type=Audio}
 UE.Source.OneShot.OnFinished {type=Trigger}
]
pages [Default {
  input_defaults {
   Boss.Phase=-1
   DuckingStrength=1.84
   FadeInTime=2
   FadeOutTime=3
   MasterVolume=1
   Mix.Ending=1.2634
   Mix.Phase0.Else=1
   Mix.Phase0.Master=1.3033
   Mix.Phase0.Perc=1
   Mix.Phase1.Else=1
   Mix.Phase1.Master=1.1437
   Mix.Phase1.Perc=1
   Mix.Phase2.Else=1
   Mix.Phase2.Master=1
   Mix.Phase2.Perc=1
   Mix.Transition0to1=1.2554
   Mix.Transition1to2=1.2315
   Music.BeatsPerBar=4
   Music.Ending=/TidalCrane_Swarm_Boss/Assets/Sound/Music/wav/MX_S37_Combat_TidalCrane_Death_STNG
   Music.Phase0.Else=/ProtoSwarm/Assets/Audio/SoundWaves/Music/MX_S37_Combat_ArmoredBug_Phase_0_LP_Else
   Music.Phase0.Perc=/ProtoSwarm/Assets/Audio/SoundWaves/Music/MX_S37_Combat_ArmoredBug_Phase_0_LP_Perc
   Music.Phase1.Else=/TidalCrane_Swarm_Boss/Assets/Sound/Music/wav/MX_S37_Combat_TidalCrane_Phase_1_LP_Else
   Music.Phase1.Perc=/TidalCrane_Swarm_Boss/Assets/Sound/Music/wav/MX_S37_Combat_TidalCrane_Phase_1_LP_Perc
   Music.Phase2.Else=/TidalCrane_Swarm_Boss/Assets/Sound/Music/wav/MX_S37_Combat_TidalCrane_Phase_2_LP_Else
   Music.Phase2.Perc=/TidalCrane_Swarm_Boss/Assets/Sound/Music/wav/MX_S37_Combat_TidalCrane_Phase_2_LP_Perc
   Music.Tempo=146
   Music.Transition0to1=/TidalCrane_Swarm_Boss/Assets/Sound/Music/wav/MX_S37_Combat_TidalCrane_Phase_0_to_1_STNG
   Music.Transition1to2=/TidalCrane_Swarm_Boss/Assets/Sound/Music/wav/MX_S37_Combat_TidalCrane_Phase_1_to_2_STNG
   MusicEncounter.Attenuation.Volume.FalloffDistance=5000
   MusicEncounter.Attenuation.Volume.InnerRadius=3000
   MusicEncounter.bIsPaused=null
   UE.Attenuation.Distance=899.9987
   UE.Source.OnPlay=null
  }
  nodes [
   InitVariable.Float()
   InitVariable.Time()
   InitVariable.Trigger()
   InitVariable.Trigger()
   InitVariable.Trigger()
   InitVariable.Trigger()
   InitVariable.Trigger()
   TriggerOnValueChange_1 = UE.Trigger On Value Change:Int32(Value=^Boss.Phase)
   n1 = VariableDeferredAccessor.Trigger(Variable=InitVariable.Trigger().Variable)
   TriggerControl_0 = UE.Trigger Control(Trigger In=@TriggerOnValueChange_1.Trigger, Close=@n1.Value)
   n2 = TriggerCompare.Clamp:Int32(A=^Boss.Phase, Compare=@TriggerControl_0.Trigger Out)
   n3 = HarmonixNodes.Metronome(Time Sig. Numerator=^Music.BeatsPerBar, Tempo=^Music.Tempo, Transport=HarmonixNodes.TriggerToTransport(Play=^UE.Source.OnPlay).Transport)
   n4 = VariableAccessor.Trigger(Variable=VariableMutator.Trigger(Value=HarmonixNodes.MidiClockSubdivisionTrigger(MIDI Clock=@n3.MIDI Clock).Trigger Out, Variable=InitVariable.Trigger().Variable).Variable)
   TriggerAnyTriggerAny2_3 = "TriggerAny.Trigger Any (2)"(
    In 0=&MSP_ScheduleTrigger(Schedule=@n2.True, In Trigger=@n4.Value).Scheduled Trigger
    In 1=TriggerCompare.Clamp:Int32(A=^Boss.Phase, Compare=^UE.Source.OnPlay).True
   )
   n5 = VariableAccessor.Trigger(Variable=VariableMutator.Trigger(Variable=InitVariable.Trigger().Variable, Value=@TriggerAnyTriggerAny2_3.Out).Variable)
   n6 = VariableAccessor.Trigger(Variable=@n5.Variable)
   n7 = TriggerCompare.Clamp:Int32(A=^Boss.Phase, Compare=@n2.False) {also {defaults {B=1}}}
   n8 = TriggerCompare.Clamp:Int32(A=^Boss.Phase, Compare=@n7.False) {also {defaults {B=2}}}
   n9 = VariableAccessor.Trigger(Variable=@n4.Variable)
   n10 = VariableAccessor.Trigger(Variable=@n9.Variable)
   n11 = VariableAccessor.Trigger(Variable=@n10.Variable)
   n12 = VariableAccessor.Trigger(Variable=@n11.Variable)
   n13 = VariableAccessor.Trigger(Variable=@n12.Variable)
   n14 = VariableAccessor.Trigger(Variable=@n13.Variable)
   TriggerAnyTriggerAny2_4 = "TriggerAny.Trigger Any (2)"(
    In 0=&MSP_ScheduleTrigger(
     Schedule=TriggerCompare.Clamp:Int32(A=^Boss.Phase, Compare=@n8.False) {also {defaults {B=3}}}.True
     In Trigger=VariableAccessor.Trigger(Variable=@n14.Variable).Value
    ).Scheduled Trigger
    In 1=TriggerCompare.Clamp:Int32(A=^Boss.Phase, Compare=^UE.Source.OnPlay) {also {defaults {B=3}}}.True
   )
   n15 = VariableAccessor.Trigger(Variable=VariableMutator.Trigger(Value=@TriggerAnyTriggerAny2_4.Out, Variable=@n1.Variable).Variable)
   n16 = VariableAccessor.Trigger(Variable=@n15.Variable)
   TriggerDelay_3 = UE.Trigger Delay(In=@n16.Value) {also {defaults {Delay Time=0.6}}}
   TriggerAnyTriggerAny2_5 = "TriggerAny.Trigger Any (2)"(
    In 0=&MSP_ScheduleTrigger(Schedule=@n7.True, In Trigger=@n10.Value).Scheduled Trigger
    In 1=TriggerCompare.Clamp:Int32(A=^Boss.Phase, Compare=^UE.Source.OnPlay) {also {defaults {B=1}}}.True
   )
   n17 = VariableAccessor.Trigger(Variable=VariableMutator.Trigger(Variable=InitVariable.Trigger().Variable, Value=@TriggerAnyTriggerAny2_5.Out).Variable)
   n18 = VariableAccessor.Trigger(Variable=VariableMutator.Trigger(
     Variable=InitVariable.Trigger().Variable
     Value=&MSP_ScheduleTrigger(In Trigger=@n9.Value, Schedule=@n17.Value) {also {defaults {Triggers From Now=8}}}.Scheduled Trigger
    ).Variable)
   n19 = VariableAccessor.Trigger(Variable=@n18.Variable)
   n20 = VariableAccessor.Trigger(Variable=@n19.Variable)
   n21 = VariableAccessor.Trigger(Variable=VariableMutator.Trigger(Variable=InitVariable.Trigger().Variable, Value=@TriggerAnyTriggerAny2_5.Out).Variable)
   TriggerAnyTriggerAny2_2 = "TriggerAny.Trigger Any (2)"(
    In 0=&MSP_ScheduleTrigger(Schedule=@n8.True, In Trigger=@n14.Value).Scheduled Trigger
    In 1=TriggerCompare.Clamp:Int32(A=^Boss.Phase, Compare=^UE.Source.OnPlay) {also {defaults {B=2}}}.True
   )
   n22 = VariableAccessor.Trigger(Variable=VariableMutator.Trigger(Variable=InitVariable.Trigger().Variable, Value=@TriggerAnyTriggerAny2_2.Out).Variable)
   n23 = VariableAccessor.Trigger(Variable=VariableMutator.Trigger(
     Variable=InitVariable.Trigger().Variable
     Value=&MSP_ScheduleTrigger(In Trigger=@n11.Value, Schedule=@n22.Value) {also {defaults {Triggers From Now=9}}}.Scheduled Trigger
    ).Variable)
   n24 = VariableAccessor.Trigger(Variable=@n23.Variable)
   n25 = VariableAccessor.Trigger(Variable=@n24.Variable)
   n26 = VariableAccessor.Trigger(Variable=VariableMutator.Trigger(Variable=InitVariable.Trigger().Variable, Value=@TriggerAnyTriggerAny2_2.Out).Variable)
   n27 = VariableAccessor.Trigger(Variable=@n26.Variable)
   TriggerAnyTriggerAny5_1 = "TriggerAny.Trigger Any (5)"(
    In 4=@TriggerDelay_3.Out
    In 1=@n20.Value
    In 0=VariableAccessor.Trigger(Variable=@n21.Variable).Value
    In 3=@n25.Value
    In 2=VariableAccessor.Trigger(Variable=@n27.Variable).Value
   )
   n28 = VariableAccessor.Trigger(Variable=VariableMutator.Trigger(Variable=InitVariable.Trigger().Variable, Value=@TriggerAnyTriggerAny5_1.Out).Variable)
   WavePlayer_0 = UE.Wave Player:Stereo(Wave Asset=^Music.Phase0.Perc, Play=@n6.Value, Stop=VariableAccessor.Trigger(Variable=@n28.Variable).Value) {also {defaults {Loop="<MetaSoundBuilderResult.SUCCEEDED: 0>"}}}
   WavePlayer_3 = UE.Wave Player:Stereo(Wave Asset=^Music.Phase0.Else, Stop=@n28.Value, Play=@n5.Value) {also {defaults {Loop="<MetaSoundBuilderResult.SUCCEEDED: 0>"}}}
   AudioMixerAudioMixerStereo2_0 = "AudioMixer.Audio Mixer (Stereo, 2)"(
    Gain 0=^Mix.Phase0.Else
    Gain 1=^Mix.Phase0.Perc
    In 1 L=@WavePlayer_0.Out Left
    In 1 R=@WavePlayer_0.Out Right
    In 0 L=@WavePlayer_3.Out Left
    In 0 R=@WavePlayer_3.Out Right
   )
   n29 = VariableAccessor.Trigger(Variable=@n16.Variable)
   TriggerDelay_2 = UE.Trigger Delay(In=@n29.Value) {also {defaults {Delay Time=0.6}}}
   n30 = VariableAccessor.Trigger(Variable=@n25.Variable)
   TriggerAnyTriggerAny5_2 = "TriggerAny.Trigger Any (5)"(
    In 2=&MSP_ScheduleTrigger(In Trigger=@n12.Value, Schedule=@n27.Value).Scheduled Trigger
    In 4=@TriggerDelay_2.Out
    In 3=@n30.Value
   )
   n31 = VariableAccessor.Trigger(Variable=VariableMutator.Trigger(Variable=InitVariable.Trigger().Variable, Value=@TriggerAnyTriggerAny5_2.Out).Variable)
   WavePlayer_2 = UE.Wave Player:Stereo(Wave Asset=^Music.Phase1.Else, Play=@n18.Value, Stop=@n31.Value) {also {defaults {Loop="<MetaSoundBuilderResult.SUCCEEDED: 0>"}}}
   WavePlayer_5 = UE.Wave Player:Stereo(Wave Asset=^Music.Phase1.Perc, Stop=VariableAccessor.Trigger(Variable=@n31.Variable).Value, Play=@n19.Value) {also {defaults {Loop="<MetaSoundBuilderResult.SUCCEEDED: 0>"}}}
   AudioMixerAudioMixerStereo2_1 = "AudioMixer.Audio Mixer (Stereo, 2)"(
    Gain 0=^Mix.Phase1.Else
    Gain 1=^Mix.Phase1.Perc
    In 0 L=@WavePlayer_2.Out Left
    In 0 R=@WavePlayer_2.Out Right
    In 1 L=@WavePlayer_5.Out Left
    In 1 R=@WavePlayer_5.Out Right
   )
   n32 = VariableAccessor.Trigger(Variable=@n29.Variable)
   TriggerDelay_1 = UE.Trigger Delay(In=@n32.Value) {also {defaults {Delay Time=0.6}}}
   TriggerAnyTriggerAny5_0 = "TriggerAny.Trigger Any (5)"(In 4=@TriggerDelay_1.Out, In 2=VariableAccessor.Trigger(Variable=@n20.Variable).Value)
   n33 = VariableAccessor.Trigger(Variable=VariableMutator.Trigger(Variable=InitVariable.Trigger().Variable, Value=@TriggerAnyTriggerAny5_0.Out).Variable)
   WavePlayer_7 = UE.Wave Player:Stereo(Wave Asset=^Music.Phase2.Else, Play=@n23.Value, Stop=@n33.Value) {also {defaults {Loop="<MetaSoundBuilderResult.SUCCEEDED: 0>"}}}
   WavePlayer_8 = UE.Wave Player:Stereo(Wave Asset=^Music.Phase2.Perc, Play=@n24.Value, Stop=VariableAccessor.Trigger(Variable=@n33.Variable).Value) {also {defaults {Loop="<MetaSoundBuilderResult.SUCCEEDED: 0>"}}}
   AudioMixerAudioMixerStereo2_2 = "AudioMixer.Audio Mixer (Stereo, 2)"(
    Gain 0=^Mix.Phase2.Else
    Gain 1=^Mix.Phase2.Perc
    In 0 L=@WavePlayer_7.Out Left
    In 0 R=@WavePlayer_7.Out Right
    In 1 L=@WavePlayer_8.Out Left
    In 1 R=@WavePlayer_8.Out Right
   )
   n34 = InitVariable.Float()
   n35 = InitVariable.Float()
   n36 = InitVariable.Float()
   n37 = Value.Value:Float(Set=&MSP_ScheduleTrigger(In Trigger=@n13.Value, Schedule=VariableAccessor.Trigger(Variable=@n32.Variable).Value).Scheduled Trigger) {also {defaults {Init Value=1}}}
   TriggerAnyTriggerAny2_0 = "TriggerAny.Trigger Any (2)"(In 1=VariableAccessor.Trigger(Variable=@n30.Variable).Value)
   n38 = InitVariable.Float()
   n39 = VariableAccessor.Trigger(Variable=@n6.Variable)
   n40 = UE.InterpTo:Audio(Target=Value.Value:Float(Reset=@TriggerAnyTriggerAny2_0.Out, Set=@n39.Value) {also {defaults {Init Value=1, Target Value=0}}}.Output Value) {also {defaults {Interp Time=3}}}
   TriggerRouteTriggerRouteFloat2_0 = "TriggerRoute.Trigger Route (Float, 2):Float"(
    Value 1=UE.InterpTo:Audio(Target=Value.Value:Float(Set=@TriggerAnyTriggerAny2_0.Out) {also {defaults {Init Value=0, Target Value=1}}}.Output Value) {also {defaults {Interp Time=0.3}}}.Value
    Set 1=@TriggerAnyTriggerAny2_0.Out
    Value 0=VariableAccessor.Float(Variable=VariableMutator.Float(Variable=@n38.Variable, Value=@n40.Value).Variable).Value
    Set 0=VariableAccessor.Trigger(Variable=@n39.Variable).Value
   )
   n41 = VariableMutator.Float(
    Variable=InitVariable.Float().Variable
    Value=UE.Multiply:Float(
     PrimaryOperand=UE.InterpTo:Audio(Target=@n37.Output Value) {also {defaults {Interp Time=0.3}}}.Value
     AdditionalOperands=@TriggerRouteTriggerRouteFloat2_0.Value
    ).Out
   )
   n42 = InitVariable.Float()
   WavePlayer_1 = UE.Wave Player:Stereo(Wave Asset=^Music.Ending, Play=@n15.Value)
   WavePlayer_4 = UE.Wave Player:Stereo(Wave Asset=^Music.Transition1to2, Play=@n26.Value, Stop=VariableAccessor.Trigger(Variable=InitVariable.Trigger().Variable).Value)
   WavePlayer_6 = UE.Wave Player:Stereo(Wave Asset=^Music.Transition0to1, Stop=VariableAccessor.Trigger(Variable=InitVariable.Trigger().Variable).Value, Play=@n21.Value)
   AudioMixerAudioMixerStereo6_0 = "AudioMixer.Audio Mixer (Stereo, 6)"(
    Gain 5=^Mix.Ending
    In 0 L=@AudioMixerAudioMixerStereo2_0.Out L
    In 0 R=@AudioMixerAudioMixerStereo2_0.Out R
    In 2 L=@AudioMixerAudioMixerStereo2_1.Out L
    In 2 R=@AudioMixerAudioMixerStereo2_1.Out R
    In 4 L=@AudioMixerAudioMixerStereo2_2.Out L
    In 4 R=@AudioMixerAudioMixerStereo2_2.Out R
    Gain 3=UE.Multiply:Float(PrimaryOperand=^Mix.Transition1to2, AdditionalOperands=VariableAccessor.Float(Variable=@n34.Variable).Value).Out
    Gain 2=UE.Multiply:Float(PrimaryOperand=^Mix.Phase1.Master, AdditionalOperands=VariableAccessor.Float(Variable=@n35.Variable).Value).Out
    Gain 1=UE.Multiply:Float(PrimaryOperand=^Mix.Transition0to1, AdditionalOperands=VariableAccessor.Float(Variable=@n36.Variable).Value).Out
    Gain 4=UE.Multiply:Float(PrimaryOperand=^Mix.Phase2.Master, AdditionalOperands=VariableAccessor.Float(Variable=@n41.Variable).Value).Out
    Gain 0=UE.Multiply:Float(PrimaryOperand=^Mix.Phase0.Master, AdditionalOperands=VariableAccessor.Float(Variable=@n42.Variable).Value).Out
    In 5 L=@WavePlayer_1.Out Left
    In 5 R=@WavePlayer_1.Out Right
    In 3 L=@WavePlayer_4.Out Left
    In 3 R=@WavePlayer_4.Out Right
    In 1 L=@WavePlayer_6.Out Left
    In 1 R=@WavePlayer_6.Out Right
   )
   n43 = MapRange.MapRange:Float(
    In Range A=^MusicEncounter.Attenuation.Volume.InnerRadius
    In Range B=UE.Add:Float(AdditionalOperands=^MusicEncounter.Attenuation.Volume.FalloffDistance, PrimaryOperand=^MusicEncounter.Attenuation.Volume.InnerRadius).Out
    In=UE.InterpTo:Audio(Target=^UE.Attenuation.Distance) {also {defaults {Interp Time=0.5}}}.Value
   ) {also {defaults {Out Range A=1, Out Range B=0}}}
   ADSREnvelopeADSREnvelope_0 = ADSR Envelope.ADSR Envelope:Float(Attack Time=^FadeInTime, Trigger Attack=^UE.Source.OnPlay) {also {defaults {Decay Time=0, Sustain Level=1}}}
   TriggerOnValueChange_0 = UE.Trigger On Value Change:Bool(Value=^MusicEncounter.bIsPaused)
   n44 = TriggerCompare.Clamp:Bool(A=^MusicEncounter.bIsPaused, Compare=@TriggerOnValueChange_0.Trigger) {also {defaults {B="<MetaSoundBuilderResult.SUCCEEDED: 0>"}}}
   n45 = Value.Value:Float(Reset=@n44.False, Set=@n44.True) {also {defaults {Init Value=1, Target Value=0}}}
   n46 = UE.Multiply:Float(
    PrimaryOperand=@ADSREnvelopeADSREnvelope_0.Out Envelope
    AdditionalOperands=UE.InterpTo:Audio(Interp Time=^FadeOutTime, Target=@n45.Output Value).Value
   )
   n47 = UE.Multiply:Float(
    AdditionalOperands=^DuckingStrength
    PrimaryOperand=Modulation.GetModulatorValue() {also {defaults {Modulator=/Game/Sounds/Modulation/EnvelopeFollowers/FortWeapon/MG_WeaponFireEnvFollower}}}.Out
   )
   n48 = UE.Multiply:Float(
    AdditionalOperands=UE.Multiply:Float(PrimaryOperand=^MasterVolume, AdditionalOperands=UE.Multiply:Float(PrimaryOperand=@n43.Out Value, AdditionalOperands=@n46.Out).Out).Out
    PrimaryOperand=UE.Subtract:Float(AdditionalOperands=UE.InterpTo:Audio(Target=@n47.Out) {also {defaults {Interp Time=1}}}.Value) {also {defaults {PrimaryOperand=1}}}.Out
   )
   Multiply_0 = UE.Multiply:Audio by Float(PrimaryOperand=@AudioMixerAudioMixerStereo6_0.Out L, AdditionalOperands=@n48.Out)
   Multiply_1 = UE.Multiply:Audio by Float(PrimaryOperand=@AudioMixerAudioMixerStereo6_0.Out R, AdditionalOperands=@n48.Out)
   TriggerDelay_0 = UE.Trigger Delay(Delay Time=^FadeOutTime, Reset=@n45.On Reset, In=@n45.On Set)
   TriggerAnyTriggerAny2_1 = "TriggerAny.Trigger Any (2)"(In 1=@TriggerDelay_0.Out, In 0=@WavePlayer_1.On Finished)
   n49 = TriggerCompare.Clamp:Int32(A=^Boss.Phase, Compare=^UE.Source.OnPlay) {also {defaults {Type=3}}}
   VariableMutator.Trigger(
    Value=HarmonixNodes.MidiClockSubdivisionTrigger(
     MIDI Clock=@n3.MIDI Clock
     Offset Multiplier=Value.Value:Int32(Reset=@n49.False, Set=@n49.True) {also {defaults {Init Value=3}}}.Output Value
    ) {also {defaults {Base Grid Size=0}}}.Trigger Out
    Variable=InitVariable.Trigger().Variable
   )
  ]
  links [
   {dst="@UE.OutputFormat.Stereo.Audio:0", src=Multiply_0.Out}
   {dst="@UE.OutputFormat.Stereo.Audio:1", src=Multiply_1.Out}
   {dst="@UE.Source.OneShot.OnFinished", src=TriggerAnyTriggerAny2_1.Out}
  ]
 }]
properties {
 assetImportData=&self:AssetImportData
 numChannels=2
 outputFormat=Stereo
 sampleRate=48000
 soundClassObject=&Fort_Music_NPC
}
