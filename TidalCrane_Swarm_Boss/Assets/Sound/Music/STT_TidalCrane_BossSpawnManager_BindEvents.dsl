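; State Tree task for the TidalCrane boss, filed under music, that relays boss spawn manager events into State Tree delegates.
; On enter state, if the Boss Spawn Manager from its context is valid, it binds to the manager's OnBossSet and OnEventFinished dispatchers, and unbinds both on exit.
; Boss set broadcasts the On Boss Spawn delegate, and event finished broadcasts On Boss Death, without passing the boss reference along.
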(event EventEnterState
  (params (Transition :type State%20Tree%20Transition%20Result%20Structure%20%28by%20ref%29))
  (bind _boss_spawn_manager (Variables|Context|GetBoss%20Spawn%20Manager)
    (:then
      (Default|BindEventtoOnBossSet _boss_spawn_manager (EventDispatchers|CreateEvent :!event Custom|On%20Boss%20Set))
      (EventDispatchers|BindEventtoOnEventFinished (Variables|Context|GetBoss%20Spawn%20Manager) (EventDispatchers|CreateEvent :!event Custom|On%20Event%20Finished)))
    (:else)))

(event EventExitState
  (params (Transition :type State%20Tree%20Transition%20Result%20Structure%20%28by%20ref%29))
  (bind _boss_spawn_manager (Variables|Context|GetBoss%20Spawn%20Manager)
    (:then
      (Default|UnbindEventfromOnBossSet _boss_spawn_manager (EventDispatchers|CreateEvent :!event Custom|On%20Boss%20Set))
      (EventDispatchers|UnbindEventfromOnEventFinished (Variables|Context|GetBoss%20Spawn%20Manager) (EventDispatchers|CreateEvent :!event Custom|On%20Event%20Finished)))
    (:else)))

(event Custom|On%20Boss%20Set
  (params (Boss :type Building%20Gameplay%20Actor%20Object%20Reference))
  (StateTree|BroadcastDelegate (Variables|Delegates|GetOn%20Boss%20Spawn)))

(event Custom|On%20Event%20Finished
  (StateTree|BroadcastDelegate (Variables|Delegates|GetOn%20Boss%20Death)))
