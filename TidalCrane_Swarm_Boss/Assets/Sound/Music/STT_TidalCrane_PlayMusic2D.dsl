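; StateTree task that plays 2D encounter music for the TidalCrane boss on a supplied audio component, only where cosmetic events can run.
; On enter, it unpauses already-playing audio via a MusicEncounter boolean parameter, or sets and plays the input Music, registers it as a Boss music event and binds audio-finished to an output delegate.
; On exit, it pauses through that parameter rather than stopping, and unbinds the callback.
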(event EventEnterState
  (params (Transition :type State%20Tree%20Transition%20Result%20Structure%20%28by%20ref%29))
  (Networking|CanExecuteCosmeticEvents
    (:True
      (bind _audio_component (Variables|Input|GetAudio%20Component)
        (:then
          (if (Audio|Components|Audio|IsPlaying _audio_component)
            (Audio|Parameter|SetBooleanParameter (Variables|Input|GetAudio%20Component) "MusicEncounter.bIsPaused" false :!class /Script/AudioExtensions.AudioParameterControllerInterface)
            (StateTree|FinishTask true)
            (else
              (bind _music (Variables|Input|GetMusic)
                (:then
                  (Audio|Components|Audio|SetSound (Variables|Input|GetAudio%20Component) _music)
                  (Audio|Components|Audio|Play (Variables|Input|GetAudio%20Component) 0.0)
                  (Audio|AddEventbyComponent (GameInstanceSubsystems|GetMusicEventSubsystem) (Variables|Input|GetAudio%20Component) "(TagName=\"MusicEventSubsystem.Event.Boss\")" "(GameplayTags=)" 0 "")
                  (EventDispatchers|BindEventtoOnAudioFinished (Variables|Input|GetAudio%20Component) (EventDispatchers|CreateEvent :!event Custom|On%20Finished))
                  (StateTree|FinishTask true))
                (:else
                  (StateTree|FinishTask false))))))
        (:else)))
    (:False
      (StateTree|FinishTask true))))

(event EventExitState
  (params (Transition :type State%20Tree%20Transition%20Result%20Structure%20%28by%20ref%29))
  (bind _audio_component (Variables|Input|GetAudio%20Component)
    (:then
      (Audio|Parameter|SetBooleanParameter _audio_component "MusicEncounter.bIsPaused" true :!class /Script/AudioExtensions.AudioParameterControllerInterface)
      (EventDispatchers|UnbindEventfromOnAudioFinished (Variables|Input|GetAudio%20Component) (EventDispatchers|CreateEvent :!event Custom|On%20Finished)))
    (:else)))

(event Custom|On%20Finished
  (StateTree|BroadcastDelegate (Variables|Output|GetOn%20Audio%20Finished)))
