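# StateTree that drives TidalCrane boss music from spawn-manager events and whether the listener is inside an Audio Gameplay Volume.
# Global tasks bind those events and add an audio component; the tree succeeds once the event finishes.
# Inside the volume, it plays 2D music, sets radius and falloff parameters, and selects Idle, Combat or Ending.
# Integer parameters set the music phase, with Combat moving from Phase 1 to 2 below a building-health threshold.
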StateTree /TidalCrane_Swarm_Boss/Assets/Sound/Music/ST_TidalCrane_BossMusic
global_tasks [
 {
  id=#1
  description=Bind to TidalCrane Boss Spawn Manager
  properties {
   iD=#2
   instanceObject=&self:StateTreeEditorData_0.STT_TidalCrane_BossSpawnManager_BindEvents_C_1
   node {
    name=None
    _structType=&StateTreeBlueprintTaskWrapper
    taskClass=&/STT_TidalCrane_BossSpawnManager_BindEvents.STT_TidalCrane_BossSpawnManager_BindEvents_C
   }
  }
 }
 {
  id=#3
  description=Add Audio Component
  properties {
   iD=#4
   instanceObject=&self:StateTreeEditorData_0.STT_AddAudioComponent_C_0
   node {
    name=None
    _structType=&StateTreeBlueprintTaskWrapper
    taskClass=/Game/Sounds/StateTreeTasks/STT_AddAudioComponent.STT_AddAudioComponent_C
   }
  }
 }
]
states [
 Root {
  id=#5
  type=STATE
  children [#6, #7]
  tasks []
  transitions []
 }
 Event is completed {
  id=#6
  type=STATE
  enter_conditions [{
    id=#8
    description=Actor.Is Event Finished is True
    properties {
     iD=#9
     instance {_structType=/Script/StateTreeModule.StateTreeCompareBoolConditionInstanceData, bLeft=false, bRight=true}
     node {
      name=None
      _structType=/Script/StateTreeModule.StateTreeCompareBoolCondition
      bInvert=false
      deltaIndent=0
      evaluationMode=Evaluated
      operand=And
      bindingsBatch {value=65535}
      executionRuntimeTemplateIndex {value=65535}
      instanceDataHandle {index=65535, source=None, stateHandle {index=65535}}
      instanceTemplateIndex {value=65535}
      outputBindingsBatch {value=65535}
     }
    }
   }]
  tasks []
  transitions [{id=#10, trigger=ON_STATE_COMPLETED, state {id=#11, name=Root, linkType=SUCCEEDED}}]
 }
 Event is active {
  id=#7
  type=STATE
  children [#12, #13]
  enter_conditions [{
    id=#14
    description=Actor.Is Event Finished is False
    properties {
     iD=#15
     instance {_structType=&StateTreeCompareBoolConditionInstanceData, bLeft=false, bRight=false}
     node {
      name=None
      _structType=&StateTreeCompareBoolCondition
      bInvert=false
      deltaIndent=0
      evaluationMode=Evaluated
      operand=And
      bindingsBatch {value=65535}
      executionRuntimeTemplateIndex {value=65535}
      instanceDataHandle {index=65535, source=None, stateHandle {index=65535}}
      instanceTemplateIndex {value=65535}
      outputBindingsBatch {value=65535}
     }
    }
   }]
  tasks [
   {
    id=#16
    description=Get Volume Proxy Component
    properties {
     iD=#17
     instanceObject=&self:StateTreeEditorData_0.StateTreeState_2.STT_GetAudioGameplayVolumeProxyComponent_C_0
     node {
      name=None
      _structType=&StateTreeBlueprintTaskWrapper
      taskClass=/Game/Sounds/StateTreeTasks/STT_GetAudioGameplayVolumeProxyComponent.STT_GetAudioGameplayVolumeProxyComponent_C
     }
    }
   }
   {
    id=#18
    description="Bind Events (AGV Proxy)"
    properties {
     iD=#19
     instanceObject=&self:StateTreeEditorData_0.StateTreeState_2.STT_AudioGameplayVolumeProxy_BindEvents_C_0
     node {
      name=None
      _structType=&StateTreeBlueprintTaskWrapper
      taskClass=/Game/Sounds/StateTreeTasks/STT_AudioGameplayVolumeProxy_BindEvents.STT_AudioGameplayVolumeProxy_BindEvents_C
     }
    }
   }
  ]
  transitions [
   {
    id=#20
    trigger=ON_DELEGATE
    conditions [{
      id=#21
      description=Actor.Is Event Finished is True
      properties {
       iD=#22
       instance {_structType=&StateTreeCompareBoolConditionInstanceData, bLeft=false, bRight=true}
       node {
        name=None
        _structType=&StateTreeCompareBoolCondition
        bInvert=false
        deltaIndent=0
        evaluationMode=Evaluated
        operand=And
        bindingsBatch {value=65535}
        executionRuntimeTemplateIndex {value=65535}
        instanceDataHandle {index=65535, source=None, stateHandle {index=65535}}
        instanceTemplateIndex {value=65535}
        outputBindingsBatch {value=65535}
       }
      }
     }]
    state {id=#6, name=Event is completed}
   }
   {
    id=#23
    trigger=ON_DELEGATE
    conditions [{
      id=#24
      description=Actor.Is Event Finished is False
      properties {
       iD=#25
       instance {_structType=&StateTreeCompareBoolConditionInstanceData, bLeft=false, bRight=false}
       node {
        name=None
        _structType=&StateTreeCompareBoolCondition
        bInvert=false
        deltaIndent=0
        evaluationMode=Evaluated
        operand=And
        bindingsBatch {value=65535}
        executionRuntimeTemplateIndex {value=65535}
        instanceDataHandle {index=65535, source=None, stateHandle {index=65535}}
        instanceTemplateIndex {value=65535}
        outputBindingsBatch {value=65535}
       }
      }
     }]
    state {id=#13, name=Listener is in AGV}
   }
   {id=#26, trigger=ON_DELEGATE, state {id=#12, name=Listener is not in AGV}}
   {id=#27, trigger=ON_DELEGATE, state {id=#6, name=Event is completed}}
  ]
 }
 Listener is not in AGV {
  id=#12
  type=STATE
  enter_conditions [{
    id=#28
    description="Bind Events (AGV Proxy).Is Listener In AGV is False"
    properties {
     iD=#29
     instance {_structType=&StateTreeCompareBoolConditionInstanceData, bLeft=false, bRight=false}
     node {
      name=None
      _structType=&StateTreeCompareBoolCondition
      bInvert=false
      deltaIndent=0
      evaluationMode=Evaluated
      operand=And
      bindingsBatch {value=65535}
      executionRuntimeTemplateIndex {value=65535}
      instanceDataHandle {index=65535, source=None, stateHandle {index=65535}}
      instanceTemplateIndex {value=65535}
      outputBindingsBatch {value=65535}
     }
    }
   }]
  tasks []
  transitions []
 }
 Listener is in AGV {
  id=#13
  type=STATE
  children [#30, #31, #32]
  enter_conditions [{
    id=#33
    description="Bind Events (AGV Proxy).Is Listener In AGV is True"
    properties {
     iD=#34
     instance {_structType=&StateTreeCompareBoolConditionInstanceData, bLeft=false, bRight=true}
     node {
      name=None
      _structType=&StateTreeCompareBoolCondition
      bInvert=false
      deltaIndent=0
      evaluationMode=Evaluated
      operand=And
      bindingsBatch {value=65535}
      executionRuntimeTemplateIndex {value=65535}
      instanceDataHandle {index=65535, source=None, stateHandle {index=65535}}
      instanceTemplateIndex {value=65535}
      outputBindingsBatch {value=65535}
     }
    }
   }]
  tasks [
   {
    id=#35
    description="Play Music 2D (TidalCrane)"
    properties {
     iD=#36
     instanceObject=&self:StateTreeEditorData_0.StateTreeState_4.STT_TidalCrane_PlayMusic2D_C_0
     node {
      name=None
      _structType=&StateTreeBlueprintTaskWrapper
      taskClass=&/STT_TidalCrane_PlayMusic2D.STT_TidalCrane_PlayMusic2D_C
     }
    }
   }
   {
    id=#37
    description=Set Inner Radius
    properties {
     iD=#38
     instanceObject=&self:StateTreeEditorData_0.StateTreeState_4.STT_AudioComponent_SetParameter_Float_C_1
     node {
      name=Set Inner Radius
      _structType=&StateTreeBlueprintTaskWrapper
      taskClass=/Game/Sounds/StateTreeTasks/STT_AudioComponent_SetParameter_Float.STT_AudioComponent_SetParameter_Float_C
     }
    }
   }
   {
    id=#39
    description=Set Falloff Distance
    properties {
     iD=#40
     instanceObject=&self:StateTreeEditorData_0.StateTreeState_4.STT_AudioComponent_SetParameter_Float_C_0
     node {
      name=Set Falloff Distance
      _structType=&StateTreeBlueprintTaskWrapper
      taskClass=&STT_AudioComponent_SetParameter_Float_C
     }
    }
   }
  ]
  transitions [
   {id=#41, trigger=ON_DELEGATE, state {id=#31, name=Combat}}
   {id=#42, trigger=ON_DELEGATE, state {id=#32, name=Ending}}
  ]
 }
 Idle {
  id=#30
  type=STATE
  enter_conditions [{
    id=#43
    description="Not  Is Object Valid"
    properties {
     iD=#44
     instance {_structType=/Script/StateTreeModule.StateTreeObjectIsValidConditionInstanceData, object=null}
     node {
      name=None
      _structType=/Script/StateTreeModule.StateTreeObjectIsValidCondition
      bInvert=true
      deltaIndent=0
      evaluationMode=Evaluated
      operand=And
      bindingsBatch {value=65535}
      executionRuntimeTemplateIndex {value=65535}
      instanceDataHandle {index=65535, source=None, stateHandle {index=65535}}
      instanceTemplateIndex {value=65535}
      outputBindingsBatch {value=65535}
     }
    }
   }]
  tasks [{
    id=#45
    description=Set Idle Phase Index
    properties {
     iD=#46
     instanceObject=&self:StateTreeEditorData_0.StateTreeState_5.STT_AudioComponent_SetParameter_Integer_C_0
     node {
      name=Set Idle Phase Index
      _structType=&StateTreeBlueprintTaskWrapper
      taskClass=/Game/Sounds/StateTreeTasks/STT_AudioComponent_SetParameter_Integer.STT_AudioComponent_SetParameter_Integer_C
     }
    }
   }]
  transitions []
 }
 Combat {
  id=#31
  type=STATE
  children [#47, #48]
  enter_conditions [{
    id=#49
    description=Is Object Valid
    properties {
     iD=#50
     instance {_structType=&StateTreeObjectIsValidConditionInstanceData, object=null}
     node {
      name=None
      _structType=&StateTreeObjectIsValidCondition
      bInvert=false
      deltaIndent=0
      evaluationMode=Evaluated
      operand=And
      bindingsBatch {value=65535}
      executionRuntimeTemplateIndex {value=65535}
      instanceDataHandle {index=65535, source=None, stateHandle {index=65535}}
      instanceTemplateIndex {value=65535}
      outputBindingsBatch {value=65535}
     }
    }
   }]
  tasks []
  transitions [{id=#51, trigger=ON_DELEGATE, state {id=#32, name=Ending}}]
 }
 Ending {
  id=#32
  type=STATE
  tasks [{
    id=#52
    description=Set Ending Phase Index
    properties {
     iD=#53
     instanceObject=&self:StateTreeEditorData_0.StateTreeState_7.STT_AudioComponent_SetParameter_Integer_C_0
     node {
      name=Set Ending Phase Index
      _structType=&StateTreeBlueprintTaskWrapper
      taskClass=&STT_AudioComponent_SetParameter_Integer_C
     }
    }
   }]
  transitions [{id=#54, trigger=ON_DELEGATE, state {id=#6, name=Event is completed}}]
 }
 Phase1 {
  id=#47
  type=STATE
  enter_conditions [{
    id=#55
    description=Is Building Health Percent Greater Than or Equal To Threshold
    properties {
     iD=#56
     instanceObject=&self:StateTreeEditorData_0.StateTreeState_8.STC_BuildingGameplayActor_IsHealthPercentBelowThreshold_C_0
     node {
      name=None
      _structType=/Script/StateTreeModule.StateTreeBlueprintConditionWrapper
      conditionClass=/Game/Sounds/StateTreeTasks/Conditions/STC_BuildingGameplayActor_IsHealthPercentBelowThreshold.STC_BuildingGameplayActor_IsHealthPercentBelowThreshold_C
      deltaIndent=0
      evaluationMode=Evaluated
      operand=And
      bindingsBatch {value=65535}
      executionRuntimeTemplateIndex {value=65535}
      instanceDataHandle {index=65535, source=None, stateHandle {index=65535}}
      instanceTemplateIndex {value=65535}
      outputBindingsBatch {value=65535}
     }
    }
   }]
  tasks [{
    id=#57
    description=Set Parameter Integer
    properties {
     iD=#58
     instanceObject=&self:StateTreeEditorData_0.StateTreeState_8.STT_AudioComponent_SetParameter_Integer_C_0
     node {
      name=None
      _structType=&StateTreeBlueprintTaskWrapper
      taskClass=&STT_AudioComponent_SetParameter_Integer_C
     }
    }
   }]
  transitions [{
    id=#59
    trigger=ON_TICK
    conditions [{
      id=#60
      description=Is Building Health Percent Less Than Threshold
      properties {
       iD=#61
       instanceObject=&self:StateTreeEditorData_0.StateTreeState_8.STC_BuildingGameplayActor_IsHealthPercentBelowThreshold_C_1
       node {
        name=None
        _structType=&StateTreeBlueprintConditionWrapper
        conditionClass=&STC_BuildingGameplayActor_IsHealthPercentBelowThreshold_C
        deltaIndent=0
        evaluationMode=Evaluated
        operand=And
        bindingsBatch {value=65535}
        executionRuntimeTemplateIndex {value=65535}
        instanceDataHandle {index=65535, source=None, stateHandle {index=65535}}
        instanceTemplateIndex {value=65535}
        outputBindingsBatch {value=65535}
       }
      }
     }]
    state {id=#48, name=Phase 2}
   }]
 }
 Phase 2 {
  id=#48
  type=STATE
  enter_conditions [{
    id=#62
    description=Is Building Health Percent Less Than Threshold
    properties {
     iD=#63
     instanceObject=&self:StateTreeEditorData_0.StateTreeState_9.STC_BuildingGameplayActor_IsHealthPercentBelowThreshold_C_0
     node {
      name=None
      _structType=&StateTreeBlueprintConditionWrapper
      conditionClass=&STC_BuildingGameplayActor_IsHealthPercentBelowThreshold_C
      deltaIndent=0
      evaluationMode=Evaluated
      operand=And
      bindingsBatch {value=65535}
      executionRuntimeTemplateIndex {value=65535}
      instanceDataHandle {index=65535, source=None, stateHandle {index=65535}}
      instanceTemplateIndex {value=65535}
      outputBindingsBatch {value=65535}
     }
    }
   }]
  tasks [{
    id=#64
    description=Set Parameter Integer
    properties {
     iD=#65
     instanceObject=&self:StateTreeEditorData_0.StateTreeState_9.STT_AudioComponent_SetParameter_Integer_C_0
     node {
      name=None
      _structType=&StateTreeBlueprintTaskWrapper
      taskClass=&STT_AudioComponent_SetParameter_Integer_C
     }
    }
   }]
  transitions []
 }
]
